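# Physics asset for the dynamic jacket of the medium male Hyde player body, defining simulated coat-tail collision and joints.
# Seven mostly box-bodied bone chains hang from pelvis-linked MSTR bodies, with the center-back chain splitting into left and right tips.
# Cross-constraints, many angularly free, tie neighboring chains together at mid and lower segments.
# Pelvis, lower spine and leg shapes approximate the body, alongside a large root-linked dyn_slope_foot_c sphere and slab.
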PhysicsAsset /Game/Characters/Player/Male/Medium/Bodies/M_MED_Hyde/Meshes/M_MED_Hyde_Jacket_Dyn
bodies {
 calf_l [
  {shapeType=Sphere, radius=15, center=<-31.8298 -0.0001 0>}
  {shapeType=Capsule, length=39.4375, radius=10.375, center=<-24.6324 2.8287 -0.0002>, rotation=<rot 88.574 0 0>}
 ]
 calf_r [
  {shapeType=Sphere, radius=15, center=<31.8298 -0.0001 0>}
  {shapeType=Capsule, length=39.4375, radius=10.375, center=<24.6324 2.8287 -0.0002>, rotation=<rot 88.574 0 0>}
 ]
 dyn_coat_bk_01 [{shapeType=Box, extentX=13.4375, extentY=14.9375, extentZ=1.5625, center=<5.954 0 -0.0003>}]
 dyn_coat_bk_02 [{shapeType=Box, extentX=13.4375, extentY=14.9375, extentZ=1.5625, center=<5.954 0 -0.0003>}]
 dyn_coat_bk_03 [{shapeType=Box, extentX=13.4375, extentY=14.9375, extentZ=1.5625, center=<5.954 0 -0.0003>}]
 dyn_coat_bk_04 [{shapeType=Box, extentX=13.4375, extentY=14.9375, extentZ=1.5625, center=<5.954 0 -0.0003>}]
 dyn_coat_bk_MSTR [{shapeType=Capsule, length=0.2, radius=0.2}]
 dyn_coat_bk_l_01 [{shapeType=Box, extentX=13.4375, extentY=14.9375, extentZ=1.5625, center=<5.954 0 -0.0003>}]
 dyn_coat_bk_l_02 [{shapeType=Box, extentX=13.4375, extentY=14.9375, extentZ=1.5625, center=<5.954 0 -0.0003>}]
 dyn_coat_bk_l_03 [{shapeType=Box, extentX=13.4375, extentY=14.9375, extentZ=1.5625, center=<5.954 0 -0.0003>}]
 dyn_coat_bk_l_04 [{shapeType=Box, extentX=13.4375, extentY=22.5625, extentZ=1.5625, center=<5.954 0 -0.0003>}]
 dyn_coat_bk_l_05 [{shapeType=Box, extentX=13.4375, extentY=26.5, extentZ=1.5625, center=<5.954 0 -0.0003>}]
 dyn_coat_bk_l_06 [{shapeType=Box, extentX=25.9375, extentY=33.5, extentZ=1.5625, center=<12.8496 0 0>, rotation=<rot 0 0 10.0003>}]
 dyn_coat_bk_l_MSTR [{shapeType=Capsule, length=0.2, radius=0.2}]
 dyn_coat_bk_r_01 [{shapeType=Box, extentX=13.4375, extentY=14.9375, extentZ=1.5625, center=<5.954 0 -0.0003>}]
 dyn_coat_bk_r_02 [{shapeType=Box, extentX=13.4375, extentY=14.9375, extentZ=2.5625, center=<5.954 0 -0.0003>}]
 dyn_coat_bk_r_03 [{shapeType=Box, extentX=13.4375, extentY=23.4375, extentZ=2.5625, center=<5.954 0 -0.5431>}]
 dyn_coat_bk_r_04 [{shapeType=Box, extentX=17.9375, extentY=26.9375, extentZ=3.5625, center=<5.954 0.0109 -0.4091>, rotation=<rot 0 0 -1.5235>}]
 dyn_coat_bk_r_05 [{shapeType=Box, extentX=25.4375, extentY=30.375, extentZ=8.625, center=<10 0 0>}]
 dyn_coat_bk_r_06 [{shapeType=Box, extentX=18.9375, extentY=31.875, extentZ=6.125, center=<16.2732 0.0002 -1.1486>, rotation=<rot -3.6319 0 0>}]
 dyn_coat_bk_r_MSTR [{shapeType=Capsule, length=0.2, radius=0.2}]
 dyn_coat_bkl_05 [{shapeType=Box, extentX=13.4375, extentY=22.75, extentZ=2.0625, center=<5.954 0 -0.0003>, rotation=<rot 0 0 4.9656>}]
 dyn_coat_bkl_06 [{shapeType=Capsule, length=21.25, radius=6.75, center=<17.25 5.75 0>, rotation=<rot -90 -0.0074 0.0074>}]
 dyn_coat_bkr_05 [{shapeType=Box, extentX=27.9375, extentY=12.75, extentZ=5.0625, center=<17.6219 0.0001 0.3868>}]
 dyn_coat_bkr_06 [{shapeType=Capsule, length=9.25, radius=4, center=<6.2771 -0.0002 0>, rotation=<rot -90 -2.5305 2.5305>}]
 dyn_coat_fr_l_01 [{shapeType=Box, extentX=13.4375, extentY=14.9375, extentZ=1.5625, center=<5.954 0 -0.0003>}]
 dyn_coat_fr_l_02 [{shapeType=Box, extentX=13.4375, extentY=14.9375, extentZ=1.5625, center=<5.954 0 -0.0003>}]
 dyn_coat_fr_l_03 [{shapeType=Box, extentX=15.4375, extentY=14.9375, extentZ=3.5625, center=<5.954 0 -0.31>}]
 dyn_coat_fr_l_04 [{shapeType=Box, extentX=15.4375, extentY=21.4375, extentZ=3.25, center=<5.7659 0 -1.25>}]
 dyn_coat_fr_l_05 [{shapeType=Box, extentX=18.4375, extentY=21.4375, extentZ=4.25, center=<5.954 0 -1.4>}]
 dyn_coat_fr_l_06 [{shapeType=Capsule, length=15, radius=7.75, center=<12 -1 2.5>, rotation=<rot -90 0.0132 -0.0131>}]
 dyn_coat_fr_l_MSTR [{shapeType=Capsule, length=0.2, radius=0.2}]
 dyn_coat_fr_r_01 [{shapeType=Box, extentX=13.4375, extentY=14.9375, extentZ=1.5625, center=<5.954 0 -0.0003>}]
 dyn_coat_fr_r_02 [{shapeType=Box, extentX=13.4375, extentY=14.9375, extentZ=1.5625, center=<5.954 0 -0.0003>}]
 dyn_coat_fr_r_03 [{shapeType=Box, extentX=13.4375, extentY=14.9375, extentZ=1.5625, center=<5.954 0 -0.0003>}]
 dyn_coat_fr_r_04 [{shapeType=Box, extentX=13.4375, extentY=14.9375, extentZ=1.5625, center=<5.954 0 -0.0003>}]
 dyn_coat_fr_r_05 [{shapeType=Box, extentX=13.4375, extentY=14.9375, extentZ=1.5625, center=<5.954 0 -0.0003>}]
 dyn_coat_fr_r_06 [{shapeType=Box, extentX=28.9375, extentY=14.9375, extentZ=4.0625, center=<10.4312 0 -0.0002>}]
 dyn_coat_fr_r_MSTR [{shapeType=Capsule, length=0.2, radius=0.2}]
 dyn_coat_l_01 [{shapeType=Box, extentX=13.4375, extentY=14.9375, extentZ=1.5625, center=<5.954 0 -0.0003>}]
 dyn_coat_l_02 [{shapeType=Box, extentX=13.4375, extentY=14.9375, extentZ=1.5625, center=<5.954 0 -0.0003>}]
 dyn_coat_l_03 [{shapeType=Box, extentX=13.4375, extentY=14.9375, extentZ=1.5625, center=<5.954 0 -0.0003>}]
 dyn_coat_l_04 [{shapeType=Box, extentX=13.4375, extentY=21.4375, extentZ=3.25, center=<5.954 0 -0.0003>}]
 dyn_coat_l_05 [{shapeType=Box, extentX=13.4375, extentY=21.4375, extentZ=3.25, center=<5.954 0 -0.0003>}]
 dyn_coat_l_06 [{shapeType=Box, extentX=26.4375, extentY=21.4375, extentZ=3.25, center=<9.0462 0 -0.0003>}]
 dyn_coat_l_MSTR [{shapeType=Capsule, length=0.2, radius=0.2}]
 dyn_coat_r_01 [{shapeType=Box, extentX=13.4375, extentY=14.9375, extentZ=1.5625, center=<5.954 0 -0.0003>}]
 dyn_coat_r_02 [{shapeType=Box, extentX=13.4375, extentY=14.9375, extentZ=1.5625, center=<5.954 0 -0.0003>}]
 dyn_coat_r_03 [{shapeType=Box, extentX=13.4375, extentY=14.9375, extentZ=2.5625, center=<5.954 0 -0.0003>}]
 dyn_coat_r_04 [{shapeType=Box, extentX=15.4375, extentY=22.9375, extentZ=3.0625, center=<5.954 0 -0.0003>, rotation=<rot 0 0 9.4753>}]
 dyn_coat_r_05 [{shapeType=Box, extentX=21.1875, extentY=27.3125, extentZ=3.0625, center=<5.954 0 -0.0003>, rotation=<rot 0 0 10.1698>}]
 dyn_coat_r_06 [{shapeType=Box, extentX=21.1875, extentY=32.8125, extentZ=4.0625, center=<11.7064 -0.0244 2.5391>, rotation=<rot 3.9919 -0.2734 1.3829>}]
 dyn_coat_r_MSTR [{shapeType=Capsule, length=0.2, radius=0.2}]
 dyn_slope_foot_c [{shapeType=Sphere, radius=60, center=<0 0 -44.5>}, {shapeType=Box, extentX=230, extentY=230, extentZ=11.5, center=<0 0 -6.2255>}]
 pelvis [{shapeType=Capsule, length=9.5938, radius=10.4062, center=<-2.3213 0.0139 0>}]
 root [{shapeType=Box, extentX=2, extentY=2, extentZ=2, center=<0 0 -4.3204>}]
 spine_01 [{shapeType=Capsule, length=6.8438, radius=10, center=<0 0.559 0.0001>}]
 spine_02 [{shapeType=Capsule, length=6.125, radius=8.2188, center=<-0.0387 0.9601 0>}]
 thigh_l [
  {shapeType=Capsule, length=30.4688, radius=12.3438, center=<-19.2705 0.3803 0.3461>, rotation=<rot 89.7019 -178.6429 -179.9991>}
  {shapeType=Capsule, length=4.5, radius=8.75, center=<-29.6703 9.5557 0.0003>, rotation=<rot 90 0 0>}
 ]
 thigh_r [{shapeType=Capsule, length=30.4688, radius=12.3438, center=<19.2153 0.3957 0>, rotation=<rot 90 9.7524 5.817>}]
}
constraints [
 {bone1Name=dyn_coat_l_01, bone2Name=dyn_coat_l_MSTR, swing1LimitDegrees=45, swing2LimitDegrees=45, twistLimitDegrees=5}
 {bone1Name=dyn_coat_l_02, bone2Name=dyn_coat_l_01, swing1LimitDegrees=45, swing2LimitDegrees=45, twistLimitDegrees=5}
 {bone1Name=dyn_coat_l_03, bone2Name=dyn_coat_l_02, swing1LimitDegrees=45, swing2LimitDegrees=45, twistLimitDegrees=5}
 {bone1Name=dyn_coat_l_04, bone2Name=dyn_coat_l_03, swing1LimitDegrees=45, swing2LimitDegrees=45, twistLimitDegrees=5}
 {bone1Name=dyn_coat_l_05, bone2Name=dyn_coat_l_04, swing1LimitDegrees=45, swing2LimitDegrees=45, twistLimitDegrees=5}
 {bone1Name=dyn_coat_l_06, bone2Name=dyn_coat_l_05, swing1LimitDegrees=45, swing2LimitDegrees=45, twistLimitDegrees=5}
 {bone1Name=dyn_coat_fr_l_01, bone2Name=dyn_coat_fr_l_MSTR, swing1LimitDegrees=45, swing2LimitDegrees=45, twistLimitDegrees=5}
 {bone1Name=dyn_coat_fr_l_02, bone2Name=dyn_coat_fr_l_01, swing1LimitDegrees=45, swing2LimitDegrees=45, twistLimitDegrees=5}
 {bone1Name=dyn_coat_fr_l_03, bone2Name=dyn_coat_fr_l_02, swing1LimitDegrees=45, swing2LimitDegrees=45, twistLimitDegrees=5}
 {bone1Name=dyn_coat_fr_l_04, bone2Name=dyn_coat_fr_l_03, swing1LimitDegrees=45, swing2LimitDegrees=45, twistLimitDegrees=5}
 {bone1Name=dyn_coat_fr_l_05, bone2Name=dyn_coat_fr_l_04, swing1LimitDegrees=45, swing2LimitDegrees=45, twistLimitDegrees=5}
 {bone1Name=dyn_coat_fr_l_06, bone2Name=dyn_coat_fr_l_05, swing1LimitDegrees=45, swing2LimitDegrees=45, twistLimitDegrees=5}
 {bone1Name=dyn_coat_bk_01, bone2Name=dyn_coat_bk_MSTR, swing1LimitDegrees=45, swing2LimitDegrees=45, twistLimitDegrees=5}
 {bone1Name=dyn_coat_bk_02, bone2Name=dyn_coat_bk_01, swing1LimitDegrees=45, swing2LimitDegrees=45, twistLimitDegrees=5}
 {bone1Name=dyn_coat_bk_03, bone2Name=dyn_coat_bk_02, swing1LimitDegrees=45, swing2LimitDegrees=45, twistLimitDegrees=5}
 {bone1Name=dyn_coat_bk_04, bone2Name=dyn_coat_bk_03, swing1LimitDegrees=45, swing2LimitDegrees=45, twistLimitDegrees=5}
 {bone1Name=dyn_coat_bkl_05, bone2Name=dyn_coat_bk_04, swing1LimitDegrees=45, swing2LimitDegrees=45, twistLimitDegrees=5}
 {bone1Name=dyn_coat_bkl_06, bone2Name=dyn_coat_bkl_05, swing1LimitDegrees=45, swing2LimitDegrees=45, twistLimitDegrees=5}
 {bone1Name=dyn_coat_bkr_05, bone2Name=dyn_coat_bk_04, swing1LimitDegrees=45, swing2LimitDegrees=45, twistLimitDegrees=5}
 {bone1Name=dyn_coat_bkr_06, bone2Name=dyn_coat_bkr_05, swing1LimitDegrees=45, swing2LimitDegrees=45, twistLimitDegrees=5}
 {bone1Name=dyn_coat_bk_l_01, bone2Name=dyn_coat_bk_l_MSTR, swing1LimitDegrees=45, swing2LimitDegrees=45, twistLimitDegrees=5}
 {bone1Name=dyn_coat_bk_l_02, bone2Name=dyn_coat_bk_l_01, swing1LimitDegrees=45, swing2LimitDegrees=45, twistLimitDegrees=5}
 {bone1Name=dyn_coat_bk_l_03, bone2Name=dyn_coat_bk_l_02, swing1LimitDegrees=45, swing2LimitDegrees=45, twistLimitDegrees=5}
 {bone1Name=dyn_coat_bk_l_04, bone2Name=dyn_coat_bk_l_03, swing1LimitDegrees=45, swing2LimitDegrees=45, twistLimitDegrees=5}
 {bone1Name=dyn_coat_bk_l_05, bone2Name=dyn_coat_bk_l_04, swing1LimitDegrees=45, swing2LimitDegrees=45, twistLimitDegrees=5}
 {bone1Name=dyn_coat_bk_l_06, bone2Name=dyn_coat_bk_l_05, swing1LimitDegrees=45, swing2LimitDegrees=45, twistLimitDegrees=5}
 {bone1Name=dyn_coat_bk_r_01, bone2Name=dyn_coat_bk_r_MSTR, swing1LimitDegrees=45, swing2LimitDegrees=45, twistLimitDegrees=5}
 {bone1Name=dyn_coat_bk_r_02, bone2Name=dyn_coat_bk_r_01, swing1LimitDegrees=45, swing2LimitDegrees=45, twistLimitDegrees=5}
 {bone1Name=dyn_coat_bk_r_03, bone2Name=dyn_coat_bk_r_02, swing1LimitDegrees=45, swing2LimitDegrees=45, twistLimitDegrees=5}
 {bone1Name=dyn_coat_bk_r_04, bone2Name=dyn_coat_bk_r_03, swing1LimitDegrees=60, swing2LimitDegrees=60, twistLimitDegrees=5}
 {bone1Name=dyn_coat_bk_r_05, bone2Name=dyn_coat_bk_r_04, swing1LimitDegrees=45, swing2LimitDegrees=45, twistLimitDegrees=5}
 {bone1Name=dyn_coat_bk_r_06, bone2Name=dyn_coat_bk_r_05, swing1LimitDegrees=45, swing2LimitDegrees=45, twistLimitDegrees=5}
 {bone1Name=dyn_coat_fr_r_01, bone2Name=dyn_coat_fr_r_MSTR, swing1LimitDegrees=45, swing2LimitDegrees=45, twistLimitDegrees=5}
 {bone1Name=dyn_coat_fr_r_02, bone2Name=dyn_coat_fr_r_01, swing1LimitDegrees=45, swing2LimitDegrees=45, twistLimitDegrees=5}
 {bone1Name=dyn_coat_fr_r_03, bone2Name=dyn_coat_fr_r_02, swing1LimitDegrees=45, swing2LimitDegrees=45, twistLimitDegrees=5}
 {bone1Name=dyn_coat_fr_r_04, bone2Name=dyn_coat_fr_r_03, swing1LimitDegrees=45, swing2LimitDegrees=45, twistLimitDegrees=5}
 {bone1Name=dyn_coat_fr_r_05, bone2Name=dyn_coat_fr_r_04, swing1LimitDegrees=45, swing2LimitDegrees=45, twistLimitDegrees=5}
 {bone1Name=dyn_coat_fr_r_06, bone2Name=dyn_coat_fr_r_05, swing1LimitDegrees=45, swing2LimitDegrees=45, twistLimitDegrees=5}
 {bone1Name=dyn_coat_r_01, bone2Name=dyn_coat_r_MSTR, swing1LimitDegrees=45, swing2LimitDegrees=45, twistLimitDegrees=5}
 {bone1Name=dyn_coat_r_02, bone2Name=dyn_coat_r_01, swing1LimitDegrees=45, swing2LimitDegrees=45, twistLimitDegrees=5}
 {bone1Name=dyn_coat_r_03, bone2Name=dyn_coat_r_02, swing1LimitDegrees=45, swing2LimitDegrees=45, twistLimitDegrees=5}
 {bone1Name=dyn_coat_r_04, bone2Name=dyn_coat_r_03, swing1LimitDegrees=45, swing2LimitDegrees=45, twistLimitDegrees=5}
 {bone1Name=dyn_coat_r_05, bone2Name=dyn_coat_r_04, swing1LimitDegrees=45, swing2LimitDegrees=45, twistLimitDegrees=5}
 {bone1Name=dyn_coat_r_06, bone2Name=dyn_coat_r_05, swing1LimitDegrees=45, swing2LimitDegrees=45, twistLimitDegrees=5}
 {bone1Name=dyn_coat_l_MSTR, bone2Name=pelvis, swing1LimitDegrees=45, swing2LimitDegrees=45, twistLimitDegrees=5}
 {bone1Name=dyn_coat_fr_l_MSTR, bone2Name=pelvis, swing1LimitDegrees=45, swing2LimitDegrees=45, twistLimitDegrees=5}
 {bone1Name=dyn_coat_bk_MSTR, bone2Name=pelvis, swing1LimitDegrees=45, swing2LimitDegrees=45, twistLimitDegrees=5}
 {
  bone1Name=dyn_coat_bk_l_MSTR
  bone2Name=pelvis
  swing1LimitDegrees=45
  swing1Motion=Free
  swing2LimitDegrees=45
  swing2Motion=Free
  twistLimitDegrees=45
  twistMotion=Free
 }
 {
  bone1Name=dyn_coat_bk_r_MSTR
  bone2Name=pelvis
  swing1LimitDegrees=45
  swing1Motion=Free
  swing2LimitDegrees=45
  swing2Motion=Free
  twistLimitDegrees=45
  twistMotion=Free
 }
 {
  bone1Name=dyn_coat_fr_r_MSTR
  bone2Name=pelvis
  swing1LimitDegrees=45
  swing1Motion=Free
  swing2LimitDegrees=45
  swing2Motion=Free
  twistLimitDegrees=45
  twistMotion=Free
 }
 {
  bone1Name=dyn_coat_r_MSTR
  bone2Name=pelvis
  swing1LimitDegrees=45
  swing1Motion=Free
  swing2LimitDegrees=45
  swing2Motion=Free
  twistLimitDegrees=45
  twistMotion=Free
 }
 {bone1Name=spine_01, bone2Name=pelvis, swing1LimitDegrees=45, swing2LimitDegrees=45, twistLimitDegrees=5}
 {bone1Name=spine_02, bone2Name=spine_01, swing1LimitDegrees=45, swing2LimitDegrees=45, twistLimitDegrees=5}
 {
  bone1Name=thigh_l
  bone2Name=pelvis
  swing1LimitDegrees=45
  swing1Motion=Free
  swing2LimitDegrees=45
  swing2Motion=Free
  twistLimitDegrees=45
  twistMotion=Free
 }
 {
  bone1Name=calf_l
  bone2Name=thigh_l
  swing1LimitDegrees=45
  swing1Motion=Free
  swing2LimitDegrees=45
  swing2Motion=Free
  twistLimitDegrees=45
  twistMotion=Free
 }
 {
  bone1Name=thigh_r
  bone2Name=pelvis
  swing1LimitDegrees=45
  swing1Motion=Free
  swing2LimitDegrees=45
  swing2Motion=Free
  twistLimitDegrees=45
  twistMotion=Free
 }
 {
  bone1Name=calf_r
  bone2Name=thigh_r
  swing1LimitDegrees=45
  swing1Motion=Free
  swing2LimitDegrees=45
  swing2Motion=Free
  twistLimitDegrees=45
  twistMotion=Free
 }
 {
  bone1Name=pelvis
  bone2Name=root
  swing1LimitDegrees=45
  swing1Motion=Free
  swing2LimitDegrees=45
  swing2Motion=Free
  twistLimitDegrees=45
  twistMotion=Free
 }
 {
  bone1Name=dyn_coat_fr_l_04
  bone2Name=dyn_coat_l_04
  swing1LimitDegrees=45
  swing1Motion=Free
  swing2LimitDegrees=45
  swing2Motion=Free
  twistLimitDegrees=45
  twistMotion=Free
 }
 {
  bone1Name=dyn_coat_fr_l_05
  bone2Name=dyn_coat_l_05
  swing1LimitDegrees=45
  swing1Motion=Free
  swing2LimitDegrees=45
  swing2Motion=Free
  twistLimitDegrees=45
  twistMotion=Free
 }
 {
  bone1Name=dyn_coat_r_04
  bone2Name=dyn_coat_fr_r_04
  swing1LimitDegrees=45
  swing1Motion=Free
  swing2LimitDegrees=45
  swing2Motion=Free
  twistLimitDegrees=45
  twistMotion=Free
 }
 {
  bone1Name=dyn_coat_r_05
  bone2Name=dyn_coat_fr_r_05
  swing1LimitDegrees=45
  swing1Motion=Free
  swing2LimitDegrees=45
  swing2Motion=Free
  twistLimitDegrees=45
  twistMotion=Free
 }
 {bone1Name=dyn_coat_bkr_05, bone2Name=dyn_coat_bk_r_04, swing1LimitDegrees=45, swing2LimitDegrees=45, twistLimitDegrees=45}
 {bone1Name=dyn_coat_bkr_06, bone2Name=dyn_coat_bk_r_05, swing1LimitDegrees=45, swing2LimitDegrees=45, twistLimitDegrees=45}
 {
  bone1Name=dyn_coat_bkl_05
  bone2Name=dyn_coat_bk_l_04
  swing1LimitDegrees=45
  swing1Motion=Free
  swing2LimitDegrees=45
  swing2Motion=Free
  twistLimitDegrees=45
  twistMotion=Free
 }
 {
  bone1Name=dyn_coat_bkl_06
  bone2Name=dyn_coat_bk_l_05
  swing1LimitDegrees=45
  swing1Motion=Free
  swing2LimitDegrees=45
  swing2Motion=Free
  twistLimitDegrees=45
  twistMotion=Free
 }
 {
  bone1Name=dyn_coat_r_06
  bone2Name=dyn_coat_bk_r_06
  swing1LimitDegrees=45
  swing1Motion=Free
  swing2LimitDegrees=45
  swing2Motion=Free
  twistLimitDegrees=45
  twistMotion=Free
 }
 {
  bone1Name=dyn_coat_l_05
  bone2Name=dyn_coat_bk_l_05
  swing1LimitDegrees=45
  swing1Motion=Free
  swing2LimitDegrees=45
  swing2Motion=Free
  twistLimitDegrees=45
  twistMotion=Free
 }
 {bone1Name=dyn_coat_r_06, bone2Name=dyn_coat_bk_r_06, swing1LimitDegrees=45, swing2LimitDegrees=45, twistLimitDegrees=45}
 {bone1Name=dyn_coat_r_05, bone2Name=dyn_coat_bk_r_05, swing1LimitDegrees=45, swing2LimitDegrees=45, twistLimitDegrees=45}
 {
  bone1Name=dyn_coat_r_03
  bone2Name=dyn_coat_bk_r_03
  swing1LimitDegrees=45
  swing1Motion=Free
  swing2LimitDegrees=45
  swing2Motion=Free
  twistLimitDegrees=45
  twistMotion=Free
 }
 {bone1Name=dyn_coat_r_04, bone2Name=dyn_coat_bk_r_04, swing1LimitDegrees=45, swing2LimitDegrees=45, twistLimitDegrees=45}
 {bone1Name=dyn_coat_bkr_06, bone2Name=dyn_coat_bk_r_06, swing1LimitDegrees=45, swing2LimitDegrees=45, twistLimitDegrees=45}
 {
  bone1Name=dyn_slope_foot_c
  bone2Name=root
  swing1LimitDegrees=45
  swing1Motion=Free
  swing2LimitDegrees=45
  swing2Motion=Free
  twistLimitDegrees=45
  twistMotion=Free
 }
 {
  bone1Name=dyn_coat_l_06
  bone2Name=dyn_coat_bk_l_06
  swing1LimitDegrees=45
  swing1Motion=Free
  swing2LimitDegrees=45
  swing2Motion=Free
  twistLimitDegrees=45
  twistMotion=Free
 }
 {
  bone1Name=dyn_coat_bk_l_06
  bone2Name=dyn_coat_bkl_06
  swing1LimitDegrees=45
  swing1Motion=Free
  swing2LimitDegrees=45
  swing2Motion=Free
  twistLimitDegrees=45
  twistMotion=Free
 }
 {
  bone1Name=dyn_coat_fr_l_06
  bone2Name=dyn_coat_l_06
  swing1LimitDegrees=45
  swing2LimitDegrees=45
  swing2Motion=Free
  twistLimitDegrees=45
  twistMotion=Free
 }
 {
  bone1Name=dyn_coat_l_04
  bone2Name=dyn_coat_bk_l_04
  swing1LimitDegrees=45
  swing1Motion=Free
  swing2LimitDegrees=45
  swing2Motion=Free
  twistLimitDegrees=45
  twistMotion=Free
 }
 {
  bone1Name=dyn_coat_l_03
  bone2Name=dyn_coat_bk_l_03
  swing1LimitDegrees=45
  swing1Motion=Free
  swing2LimitDegrees=45
  swing2Motion=Free
  twistLimitDegrees=45
  twistMotion=Free
 }
]
