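# Animation Blueprint for the Hyde male player body, adding secondary motion to its jacket, accessories, tubes and straps.
# It copies the parent mesh pose, applies slope, collar, coat and targeting bone adjustments, then caches it for physics.
# A skydiving switch picks between two jacket RigidBody simulations, followed by accessory physics and AnimDynamics chains.
# The event graph drives corrective alphas, noisy skydiving gravity, and disables normal jacket simulation for certain vehicle drivers.
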AnimBlueprint /Game/Characters/Player/Male/Medium/Bodies/M_MED_Hyde/Meshes/M_MED_Hyde_AnimBP
target_skeleton=&/M_MED_Hyde_Skeleton
graphs [
 AnimGraph {nodes [
   n1 = RigidBody(
    ComponentPose=Local To Component(LocalPose=Use cached pose 'RBAN_input_pose'())
    Alpha=Get in vehicle rigid alpha() -> (in vehicle rigid alpha:double)
   ) {
    bClampLinearTranslationLimitToRefPose=true
    bOverrideWorldGravity=true
    overridePhysicsAsset=&/M_MED_Hyde_Jacket_Dyn
    simulationSpace=BaseBoneSpace
    alphaBoolBlend {blendOption=Linear}
    baseBoneRef {boneName=root}
    componentAppliedLinearAccClamp=<3000 3000 1000>
    componentLinearAccScale=<0.5 0.4 0.4>
    componentLinearVelScale=<0.5 0.4 0.4>
    overrideWorldGravity {z=-920}
   }
   n2 = Component To Local(ComponentPose=@n1)
   n3 = RigidBody!(
    ComponentPose=Local To Component(LocalPose=Use cached pose 'RBAN_input_pose'())
    Alpha
    OverrideWorldGravity:Vector=Get GravityOverride_b() -> (GravityOverride_b:Vector)
    ExternalForce
   ) -> (Pose) {
    bClampLinearTranslationLimitToRefPose=true
    bOverrideWorldGravity=true
    overridePhysicsAsset=&M_MED_Hyde_Jacket_Dyn
    simulationSpace=BaseBoneSpace
    alphaBoolBlend {blendOption=Linear}
    baseBoneRef {boneName=pelvis}
    componentLinearAccScale=<0.15 0.15 0.15>
    componentLinearVelScale=<0.15 0.15 0.15>
    overrideWorldGravity {z=-200}
   }
   n4 = Component To Local(ComponentPose=@n3)
   n5 = Get bIsSkyDiving()
   n6 = RigidBody!(
    ComponentPose=Local To Component(LocalPose=Blend Poses by bool(BlendPose_0=@n4, BlendPose_1=@n2, bActiveValue=@n5) {blendType=Linear})
    Alpha
    ExternalForce
   ) -> (Pose) {
    bOverrideWorldGravity=true
    overridePhysicsAsset=&/M_MED_Hyde_Accessories_Physics
    simulationSpace=BaseBoneSpace
    alphaBoolBlend {blendOption=Linear}
    baseBoneRef {boneName=root}
    componentAppliedLinearAccClamp=<2000 2000 1500>
    componentLinearAccScale=<0.5 0.5 0.8>
    componentLinearVelScale=<0.5 0.5 0.8>
    overrideWorldGravity {z=-980}
   }
   n7 = Anim Dynamics - Bone: sleeve_1_l(ComponentPose=@n6, Alpha=1, ExternalForce="0,0,0") {
    angularDampingOverride=1
    angularSpringConstant=1500
    bAngularSpring=true
    bGravityOverrideInSimSpace=true
    bOverrideAngularDamping=true
    gravityScale=0.35
    simulationSpace=RootRelative
    alphaBoolBlend {blendOption=Linear}
    alphaScaleBiasClamp {scale=0.5}
    boundBone {boneName=sleeve_1_l}
    physicsBodyDefinitions [{
      sphereCollisionRadius=0
      boundBone {boneName=sleeve_1_l}
      boxExtents=<10 20 20>
      constraintSetup {angularLimitsMax=<0 60 0>, angularLimitsMin=<0 -60 -90>, angularTarget=<0 0 0>}
      localJointOffset=<10 0 0>
     }]
   }
   n8 = Anim Dynamics - Bone: sleeve_2_l(ComponentPose=@n7, Alpha=1, ExternalForce="0,0,0") {
    angularDampingOverride=1
    angularSpringConstant=1500
    bAngularSpring=true
    bGravityOverrideInSimSpace=true
    bOverrideAngularDamping=true
    gravityScale=0.35
    simulationSpace=RootRelative
    alphaBoolBlend {blendOption=Linear}
    alphaScaleBiasClamp {scale=0.5}
    boundBone {boneName=sleeve_2_l}
    physicsBodyDefinitions [{
      sphereCollisionRadius=0
      boundBone {boneName=sleeve_2_l}
      boxExtents=<10 20 20>
      constraintSetup {angularLimitsMax=<0 60 0>, angularLimitsMin=<0 -60 -90>, angularTarget=<0 0 0>}
      localJointOffset=<10 0 0>
     }]
   }
   n9 = Anim Dynamics - Bone: sleeve_3_l(ComponentPose=@n8, Alpha=1, ExternalForce="0,0,0") {
    angularDampingOverride=1
    angularSpringConstant=1500
    bAngularSpring=true
    bGravityOverrideInSimSpace=true
    bOverrideAngularDamping=true
    gravityScale=0.35
    simulationSpace=RootRelative
    alphaBoolBlend {blendOption=Linear}
    alphaScaleBiasClamp {scale=0.5}
    boundBone {boneName=sleeve_3_l}
    physicsBodyDefinitions [{
      sphereCollisionRadius=0
      boundBone {boneName=sleeve_3_l}
      boxExtents=<10 20 20>
      constraintSetup {angularLimitsMax=<0 60 0>, angularLimitsMin=<0 -60 -90>, angularTarget=<0 0 0>}
      localJointOffset=<10 0 0>
     }]
   }
   n10 = Anim Dynamics - Bone: sleeve_4_l(ComponentPose=@n9, Alpha=1, ExternalForce="0,0,0") {
    angularDampingOverride=1
    angularSpringConstant=1500
    bAngularSpring=true
    bGravityOverrideInSimSpace=true
    bOverrideAngularDamping=true
    gravityScale=0.35
    simulationSpace=RootRelative
    alphaBoolBlend {blendOption=Linear}
    alphaScaleBiasClamp {scale=0.5}
    boundBone {boneName=sleeve_4_l}
    physicsBodyDefinitions [{
      sphereCollisionRadius=0
      boundBone {boneName=sleeve_4_l}
      boxExtents=<10 20 20>
      constraintSetup {angularLimitsMax=<0 60 0>, angularLimitsMin=<0 -60 -90>, angularTarget=<0 0 0>}
      localJointOffset=<10 0 0>
     }]
   }
   n11 = Anim Dynamics - Bone: sleeve_5_l(ComponentPose=@n10, Alpha=1, ExternalForce="0,0,0") {
    angularDampingOverride=1
    angularSpringConstant=1500
    bAngularSpring=true
    bGravityOverrideInSimSpace=true
    bOverrideAngularDamping=true
    gravityScale=0.35
    simulationSpace=RootRelative
    alphaBoolBlend {blendOption=Linear}
    alphaScaleBiasClamp {scale=0.5}
    boundBone {boneName=sleeve_5_l}
    physicsBodyDefinitions [{
      sphereCollisionRadius=0
      boundBone {boneName=sleeve_5_l}
      boxExtents=<10 20 20>
      constraintSetup {angularLimitsMax=<0 60 0>, angularLimitsMin=<0 -60 -90>, angularTarget=<0 0 0>}
      localJointOffset=<10 0 0>
     }]
   }
   n12 = Anim Dynamics - Bone: sleeve_6_l(ComponentPose=@n11, Alpha=1, ExternalForce="0,0,0") {
    angularDampingOverride=1
    angularSpringConstant=1500
    bAngularSpring=true
    bGravityOverrideInSimSpace=true
    bOverrideAngularDamping=true
    gravityScale=0.35
    simulationSpace=RootRelative
    alphaBoolBlend {blendOption=Linear}
    alphaScaleBiasClamp {scale=0.5}
    boundBone {boneName=sleeve_6_l}
    physicsBodyDefinitions [{
      sphereCollisionRadius=0
      boundBone {boneName=sleeve_6_l}
      boxExtents=<10 20 20>
      constraintSetup {angularLimitsMax=<0 60 0>, angularLimitsMin=<0 -60 -90>, angularTarget=<0 0 0>}
      localJointOffset=<10 0 0>
     }]
   }
   n13 = Anim Dynamics - Bone: sleeve_7_l(ComponentPose=@n12, Alpha=1, ExternalForce="0,0,0") {
    angularDampingOverride=1
    angularSpringConstant=1500
    bAngularSpring=true
    bGravityOverrideInSimSpace=true
    bOverrideAngularDamping=true
    gravityScale=0.35
    simulationSpace=RootRelative
    alphaBoolBlend {blendOption=Linear}
    alphaScaleBiasClamp {scale=0.5}
    boundBone {boneName=sleeve_7_l}
    physicsBodyDefinitions [{
      sphereCollisionRadius=0
      boundBone {boneName=sleeve_7_l}
      boxExtents=<10 20 20>
      constraintSetup {angularLimitsMax=<0 60 0>, angularLimitsMin=<0 -60 -90>, angularTarget=<0 0 0>}
      localJointOffset=<10 0 0>
     }]
   }
   n14 = Anim Dynamics - Bone: sleeve_8_l(ComponentPose=@n13, Alpha=1, ExternalForce="0,0,0") {
    angularDampingOverride=1
    angularSpringConstant=1500
    bAngularSpring=true
    bGravityOverrideInSimSpace=true
    bOverrideAngularDamping=true
    gravityScale=0.35
    simulationSpace=RootRelative
    alphaBoolBlend {blendOption=Linear}
    alphaScaleBiasClamp {scale=0.5}
    boundBone {boneName=sleeve_8_l}
    physicsBodyDefinitions [{
      sphereCollisionRadius=0
      boundBone {boneName=sleeve_8_l}
      boxExtents=<10 20 20>
      constraintSetup {angularLimitsMax=<0 60 0>, angularLimitsMin=<0 -60 -90>, angularTarget=<0 0 0>}
      localJointOffset=<10 0 0>
     }]
   }
   n15 = Anim Dynamics - Bone: sleeve_9_l(ComponentPose=@n14, Alpha=1, ExternalForce="0,0,0") {
    angularDampingOverride=1
    angularSpringConstant=1500
    bAngularSpring=true
    bGravityOverrideInSimSpace=true
    bOverrideAngularDamping=true
    gravityScale=0.35
    simulationSpace=RootRelative
    alphaBoolBlend {blendOption=Linear}
    alphaScaleBiasClamp {scale=0.5}
    boundBone {boneName=sleeve_9_l}
    physicsBodyDefinitions [{
      sphereCollisionRadius=0
      boundBone {boneName=sleeve_9_l}
      boxExtents=<10 20 20>
      constraintSetup {angularLimitsMax=<0 60 0>, angularLimitsMin=<0 -60 -90>, angularTarget=<0 0 0>}
      localJointOffset=<10 0 0>
     }]
   }
   n16 = Anim Dynamics - Bone: sleeve_10_l(ComponentPose=@n15, Alpha=1, ExternalForce="0,0,0") {
    angularDampingOverride=1
    angularSpringConstant=1500
    bAngularSpring=true
    bGravityOverrideInSimSpace=true
    bOverrideAngularDamping=true
    gravityScale=0.35
    simulationSpace=RootRelative
    alphaBoolBlend {blendOption=Linear}
    alphaScaleBiasClamp {scale=0.5}
    boundBone {boneName=sleeve_10_l}
    physicsBodyDefinitions [{
      sphereCollisionRadius=0
      boundBone {boneName=sleeve_10_l}
      boxExtents=<10 20 20>
      constraintSetup {angularLimitsMax=<0 60 0>, angularLimitsMin=<0 -60 -90>, angularTarget=<0 0 0>}
      localJointOffset=<10 0 0>
     }]
   }
   n17 = Anim Dynamics - Bone: sleeve_1_r(ComponentPose=@n16, Alpha=1, ExternalForce="0,0,0") {
    angularDampingOverride=1
    angularSpringConstant=1500
    bAngularSpring=true
    bGravityOverrideInSimSpace=true
    bOverrideAngularDamping=true
    gravityScale=0.35
    simulationSpace=RootRelative
    alphaBoolBlend {blendOption=Linear}
    alphaScaleBiasClamp {scale=0.5}
    boundBone {boneName=sleeve_1_r}
    physicsBodyDefinitions [{
      sphereCollisionRadius=0
      boundBone {boneName=sleeve_1_r}
      boxExtents=<10 20 20>
      constraintSetup {angularLimitsMax=<0 60 0>, angularLimitsMin=<0 -60 -90>, angularTarget=<0 0 0>}
      localJointOffset=<10 0 0>
     }]
   }
   n18 = Anim Dynamics - Bone: sleeve_2_r(ComponentPose=@n17, Alpha=1, ExternalForce="0,0,0") {
    angularDampingOverride=1
    angularSpringConstant=1500
    bAngularSpring=true
    bGravityOverrideInSimSpace=true
    bOverrideAngularDamping=true
    gravityScale=0.35
    simulationSpace=RootRelative
    alphaBoolBlend {blendOption=Linear}
    alphaScaleBiasClamp {scale=0.5}
    boundBone {boneName=sleeve_2_r}
    physicsBodyDefinitions [{
      sphereCollisionRadius=0
      boundBone {boneName=sleeve_2_r}
      boxExtents=<10 20 20>
      constraintSetup {angularLimitsMax=<0 60 0>, angularLimitsMin=<0 -60 -90>, angularTarget=<0 0 0>}
      localJointOffset=<10 0 0>
     }]
   }
   n19 = Anim Dynamics - Bone: sleeve_3_r(ComponentPose=@n18, Alpha=1, ExternalForce="0,0,0") {
    angularDampingOverride=1
    angularSpringConstant=1500
    bAngularSpring=true
    bGravityOverrideInSimSpace=true
    bOverrideAngularDamping=true
    gravityScale=0.35
    simulationSpace=RootRelative
    alphaBoolBlend {blendOption=Linear}
    alphaScaleBiasClamp {scale=0.5}
    boundBone {boneName=sleeve_3_r}
    physicsBodyDefinitions [{
      sphereCollisionRadius=0
      boundBone {boneName=sleeve_3_r}
      boxExtents=<10 20 20>
      constraintSetup {angularLimitsMax=<0 60 0>, angularLimitsMin=<0 -60 -90>, angularTarget=<0 0 0>}
      localJointOffset=<10 0 0>
     }]
   }
   n20 = Anim Dynamics - Bone: sleeve_4_r(ComponentPose=@n19, Alpha=1, ExternalForce="0,0,0") {
    angularDampingOverride=1
    angularSpringConstant=1500
    bAngularSpring=true
    bGravityOverrideInSimSpace=true
    bOverrideAngularDamping=true
    gravityScale=0.35
    simulationSpace=RootRelative
    alphaBoolBlend {blendOption=Linear}
    alphaScaleBiasClamp {scale=0.5}
    boundBone {boneName=sleeve_4_r}
    physicsBodyDefinitions [{
      sphereCollisionRadius=0
      boundBone {boneName=sleeve_4_r}
      boxExtents=<10 20 20>
      constraintSetup {angularLimitsMax=<0 60 0>, angularLimitsMin=<0 -60 -90>, angularTarget=<0 0 0>}
      localJointOffset=<10 0 0>
     }]
   }
   n21 = Anim Dynamics - Bone: sleeve_5_r(ComponentPose=@n20, Alpha=1, ExternalForce="0,0,0") {
    angularDampingOverride=1
    angularSpringConstant=1500
    bAngularSpring=true
    bGravityOverrideInSimSpace=true
    bOverrideAngularDamping=true
    gravityScale=0.35
    simulationSpace=RootRelative
    alphaBoolBlend {blendOption=Linear}
    alphaScaleBiasClamp {scale=0.5}
    boundBone {boneName=sleeve_5_r}
    physicsBodyDefinitions [{
      sphereCollisionRadius=0
      boundBone {boneName=sleeve_5_r}
      boxExtents=<10 20 20>
      constraintSetup {angularLimitsMax=<0 60 0>, angularLimitsMin=<0 -60 -90>, angularTarget=<0 0 0>}
      localJointOffset=<10 0 0>
     }]
   }
   n22 = Anim Dynamics - Bone: jacket_shldr_1_l(ComponentPose=@n21, Alpha=1, ExternalForce="0,0,0") {
    angularDampingOverride=1
    angularSpringConstant=1500
    bAngularSpring=true
    bGravityOverrideInSimSpace=true
    bOverrideAngularDamping=true
    gravityScale=0.35
    simulationSpace=RootRelative
    alphaBoolBlend {blendOption=Linear}
    alphaScaleBiasClamp {scale=0.6}
    boundBone {boneName=jacket_shldr_1_l}
    physicsBodyDefinitions [{
      sphereCollisionRadius=0
      boundBone {boneName=jacket_shldr_1_l}
      boxExtents=<10 20 20>
      constraintSetup {angularLimitsMax=<0 40 0>, angularLimitsMin=<0 -40 -40>, angularTarget=<0 0 0>}
      localJointOffset=<10 0 0>
     }]
   }
   n23 = Anim Dynamics - Bone: jacket_shldr_2_l(ComponentPose=@n22, Alpha=1, ExternalForce="0,0,0") {
    angularDampingOverride=1
    angularSpringConstant=1500
    bAngularSpring=true
    bGravityOverrideInSimSpace=true
    bOverrideAngularDamping=true
    gravityScale=0.35
    simulationSpace=RootRelative
    alphaBoolBlend {blendOption=Linear}
    alphaScaleBiasClamp {scale=0.6}
    boundBone {boneName=jacket_shldr_2_l}
    physicsBodyDefinitions [{
      sphereCollisionRadius=0
      boundBone {boneName=jacket_shldr_2_l}
      boxExtents=<10 20 20>
      constraintSetup {angularLimitsMax=<0 40 0>, angularLimitsMin=<0 -40 -40>, angularTarget=<0 0 0>}
      localJointOffset=<10 0 0>
     }]
   }
   n24 = Anim Dynamics - Bone: jacket_shldr_3_l(ComponentPose=@n23, Alpha=1, ExternalForce="0,0,0") {
    angularDampingOverride=1
    angularSpringConstant=1500
    bAngularSpring=true
    bGravityOverrideInSimSpace=true
    bOverrideAngularDamping=true
    gravityScale=0.35
    simulationSpace=RootRelative
    alphaBoolBlend {blendOption=Linear}
    alphaScaleBiasClamp {scale=0.6}
    boundBone {boneName=jacket_shldr_3_l}
    physicsBodyDefinitions [{
      sphereCollisionRadius=0
      boundBone {boneName=jacket_shldr_3_l}
      boxExtents=<10 20 20>
      constraintSetup {angularLimitsMax=<0 40 0>, angularLimitsMin=<0 -40 -40>, angularTarget=<0 0 0>}
      localJointOffset=<10 0 0>
     }]
   }
   n25 = Anim Dynamics - Bone: jacket_shldr_4_l(ComponentPose=@n24, Alpha=1, ExternalForce="0,0,0") {
    angularDampingOverride=1
    angularSpringConstant=1500
    bAngularSpring=true
    bGravityOverrideInSimSpace=true
    bOverrideAngularDamping=true
    gravityScale=0.35
    simulationSpace=RootRelative
    alphaBoolBlend {blendOption=Linear}
    alphaScaleBiasClamp {scale=0.6}
    boundBone {boneName=jacket_shldr_4_l}
    physicsBodyDefinitions [{
      sphereCollisionRadius=0
      boundBone {boneName=jacket_shldr_4_l}
      boxExtents=<10 20 20>
      constraintSetup {angularLimitsMax=<0 40 0>, angularLimitsMin=<0 -40 -40>, angularTarget=<0 0 0>}
      localJointOffset=<10 0 0>
     }]
   }
   n26 = Anim Dynamics - Bone: jacket_shldr_1_r(ComponentPose=@n25, Alpha=1, ExternalForce="0,0,0") {
    angularDampingOverride=1
    angularSpringConstant=1500
    bAngularSpring=true
    bGravityOverrideInSimSpace=true
    bOverrideAngularDamping=true
    gravityScale=0.35
    simulationSpace=RootRelative
    alphaBoolBlend {blendOption=Linear}
    alphaScaleBiasClamp {scale=0.6}
    boundBone {boneName=jacket_shldr_1_r}
    physicsBodyDefinitions [{
      sphereCollisionRadius=0
      boundBone {boneName=jacket_shldr_1_r}
      boxExtents=<10 20 20>
      constraintSetup {angularLimitsMax=<0 40 0>, angularLimitsMin=<0 -40 -40>, angularTarget=<0 0 0>}
      localJointOffset=<10 0 0>
     }]
   }
   n27 = Anim Dynamics - Bone: jacket_shldr_2_r(ComponentPose=@n26, Alpha=1, ExternalForce="0,0,0") {
    angularDampingOverride=1
    angularSpringConstant=1500
    bAngularSpring=true
    bGravityOverrideInSimSpace=true
    bOverrideAngularDamping=true
    gravityScale=0.35
    simulationSpace=RootRelative
    alphaBoolBlend {blendOption=Linear}
    alphaScaleBiasClamp {scale=0.6}
    boundBone {boneName=jacket_shldr_2_r}
    physicsBodyDefinitions [{
      sphereCollisionRadius=0
      boundBone {boneName=jacket_shldr_2_r}
      boxExtents=<10 20 20>
      constraintSetup {angularLimitsMax=<0 40 0>, angularLimitsMin=<0 -40 -40>, angularTarget=<0 0 0>}
      localJointOffset=<10 0 0>
     }]
   }
   n28 = Anim Dynamics - Bone: jacket_shldr_3_r(ComponentPose=@n27, Alpha=1, ExternalForce="0,0,0") {
    angularDampingOverride=1
    angularSpringConstant=1500
    bAngularSpring=true
    bGravityOverrideInSimSpace=true
    bOverrideAngularDamping=true
    gravityScale=0.35
    simulationSpace=RootRelative
    alphaBoolBlend {blendOption=Linear}
    alphaScaleBiasClamp {scale=0.6}
    boundBone {boneName=jacket_shldr_3_r}
    physicsBodyDefinitions [{
      sphereCollisionRadius=0
      boundBone {boneName=jacket_shldr_3_r}
      boxExtents=<10 20 20>
      constraintSetup {angularLimitsMax=<0 40 0>, angularLimitsMin=<0 -40 -40>, angularTarget=<0 0 0>}
      localJointOffset=<10 0 0>
     }]
   }
   n29 = Anim Dynamics - Bone: jacket_shldr_4_r(ComponentPose=@n28, Alpha=1, ExternalForce="0,0,0") {
    angularDampingOverride=1
    angularSpringConstant=1500
    bAngularSpring=true
    bGravityOverrideInSimSpace=true
    bOverrideAngularDamping=true
    gravityScale=0.35
    simulationSpace=RootRelative
    alphaBoolBlend {blendOption=Linear}
    alphaScaleBiasClamp {scale=0.6}
    boundBone {boneName=jacket_shldr_4_r}
    physicsBodyDefinitions [{
      sphereCollisionRadius=0
      boundBone {boneName=jacket_shldr_4_r}
      boxExtents=<10 20 20>
      constraintSetup {angularLimitsMax=<0 40 0>, angularLimitsMin=<0 -40 -40>, angularTarget=<0 0 0>}
      localJointOffset=<10 0 0>
     }]
   }
   n30 = "Constraint - wrist_strap_1 (PPPP)"!(
    ComponentPose=@n29
    Alpha=1
    ConstraintWeights_0=0.2945
    ConstraintWeights_1:float=0.0303
    ConstraintWeights_2:float=0.0586
    ConstraintWeights_3:float=0.6166
   ) -> (Pose) {
    alphaBoolBlend {blendOption=Linear}
    boneToModify {boneName=wrist_strap_1}
    constraintSetup [
     {targetBone {boneName=lowerarm_twist_02_r}}
     {targetBone {boneName=pinky_metacarpal_r}}
     {targetBone {boneName=lowerarm_r}}
     {targetBone {boneName=lowerarm_twist_01_r}}
    ]
    constraintWeights [1,1,1,1]
   }
   n31 = Anim Dynamics - Bone: wrist_strap_1(ComponentPose=@n30, Alpha=1, ExternalForce="0,0,0") {
    angularDampingOverride=1
    angularSpringConstant=1500
    bAngularSpring=true
    bGravityOverrideInSimSpace=true
    bOverrideAngularDamping=true
    gravityScale=0.35
    simulationSpace=RootRelative
    alphaBoolBlend {blendOption=Linear}
    boundBone {boneName=wrist_strap_1}
    physicsBodyDefinitions [{
      sphereCollisionRadius=0
      boundBone {boneName=wrist_strap_1}
      boxExtents=<10 20 20>
      constraintSetup {angularLimitsMax=<10 90 90>, angularLimitsMin=<-10 -90 -90>, angularTarget=<0 0 0>}
      localJointOffset=<10 0 0>
     }]
   }
   n32 = Anim Dynamics - Bone: wrist_strap_2(ComponentPose=@n31, Alpha=1, ExternalForce="0,0,0") {
    angularDampingOverride=1
    angularSpringConstant=1500
    bAngularSpring=true
    bGravityOverrideInSimSpace=true
    bOverrideAngularDamping=true
    gravityScale=0.35
    simulationSpace=RootRelative
    alphaBoolBlend {blendOption=Linear}
    boundBone {boneName=wrist_strap_2}
    physicsBodyDefinitions [{
      sphereCollisionRadius=0
      boundBone {boneName=wrist_strap_2}
      boxExtents=<10 20 20>
      constraintSetup {angularLimitsMax=<10 90 90>, angularLimitsMin=<-10 -90 -90>, angularTarget=<0 0 0>}
      localJointOffset=<10 0 0>
     }]
   }
   Output Pose(Result=Component To Local(ComponentPose=@n32)) {layerGroup=None}
   n33 = Apply Additive(
    Base=Use cached pose 'InputCache'()
    Additive="Sequence Player (additive) 'M_MED_Hyde_PhysReset_Seq'"() {sequence=&/M_MED_Hyde_PhysReset_Seq}
    Alpha=Get add_debug() -> (add_debug:double)
   ) {alphaBoolBlend {blendOption=Linear}}
   n34 = Use cached pose 'InputCache'()
   n35 = Get bRigidBodyEnabled()
   n36 = "Transform (Modify) Bone - Bone: dyn_slope_foot_c"!(
    ComponentPose=Local To Component(LocalPose=Blend Poses by bool(BlendPose_0=@n33, BlendPose_1=@n34, bActiveValue=@n35) {blendType=Linear})
    bAlphaBoolEnabled:bool=NOT Boolean(A=Get bCanPlayCustomAnimations() -> (bCanPlayCustomAnimations:bool))
    Translation=Get slope_foot_trans() -> (slope_foot_trans:Vector)
    Rotation=Get slope_foot_rot() -> (slope_foot_rot:Rotator)
    Scale="1,1,1"
   ) -> (Pose) {
    alphaInputType=Bool
    rotationMode=BMM_Replace
    rotationSpace=BCS_WorldSpace
    translationMode=BMM_Replace
    translationSpace=BCS_WorldSpace
    alphaBoolBlend {blendOption=Linear}
    boneToModify {boneName=dyn_slope_foot_c}
   }
   n37 = "Transform (Modify) Bone - Bone: dyn_coat_fr_l_MSTR"(ComponentPose=@n36, Alpha=Get thigh_fwd_l_alpha() -> (thigh_fwd_l_alpha:double), Translation="20,0,5", Rotation="0,0,0", Scale="1,1,1") {
    rotationMode=BMM_Additive
    rotationSpace=BCS_BoneSpace
    translationMode=BMM_Additive
    translationSpace=BCS_ParentBoneSpace
    alphaBoolBlend {blendOption=Linear}
    alphaScaleBiasClamp {bInterpResult=true, interpSpeedDecreasing=5}
    boneToModify {boneName=dyn_coat_fr_l_MSTR}
    translation {x=20, z=5}
   }
   n38 = "Transform (Modify) Bone - Bone: collar_bk"(ComponentPose=@n37, Alpha=Get head_back_alpha() -> (head_back_alpha:double), Translation="0,0,0", Rotation="0,-20,0", Scale="1,1,1") {
    rotationMode=BMM_Additive
    rotationSpace=BCS_BoneSpace
    alphaBoolBlend {blendOption=Linear}
    alphaScaleBiasClamp {bInterpResult=true}
    boneToModify {boneName=collar_bk}
    rotation {yaw=-20}
   }
   n39 = "Transform (Modify) Bone - Bone: collar_fr_r"(ComponentPose=@n38, Alpha=Get head_right_alpha() -> (head_right_alpha:double), Translation="0,0,0", Rotation="0,0,-60", Scale="1,1,1") {
    rotationMode=BMM_Additive
    rotationSpace=BCS_BoneSpace
    alphaBoolBlend {blendOption=Linear}
    alphaScaleBiasClamp {bInterpResult=true}
    boneToModify {boneName=collar_fr_r}
    rotation {roll=-60}
   }
   n40 = "Transform (Modify) Bone - Bone: collar_fr_l"(ComponentPose=@n39, Alpha=Get head_left_alpha() -> (head_left_alpha:double), Translation="0,0,0", Rotation="0.000,0.000,60", Scale="1,1,1") {
    rotationMode=BMM_Additive
    rotationSpace=BCS_BoneSpace
    alphaBoolBlend {blendOption=Linear}
    alphaScaleBiasClamp {bInterpResult=true}
    boneToModify {boneName=collar_fr_l}
    rotation {roll=60}
   }
   n41 = "Constraint - arm_tube_hand (PPP)"!(ComponentPose=@n40, Alpha=1, ConstraintWeights_0=0.3314, ConstraintWeights_1:float=0.0588, ConstraintWeights_2:float=0.6099) -> (Pose) {
    alphaBoolBlend {blendOption=Linear}
    boneToModify {boneName=arm_tube_hand}
    constraintSetup [{targetBone {boneName=lowerarm_twist_02_r}}, {targetBone {boneName=lowerarm_r}}, {targetBone {boneName=lowerarm_twist_01_r}}]
    constraintWeights [1,1,1]
   }
   n42 = Anim Dynamics - Bone: arm_tube_hand(ComponentPose=@n41, Alpha=1, ExternalForce="0,0,0") {
    angularDampingOverride=1
    angularSpringConstant=1500
    bAngularSpring=true
    bGravityOverrideInSimSpace=true
    bOverrideAngularDamping=true
    gravityScale=0.35
    simulationSpace=RootRelative
    alphaBoolBlend {blendOption=Linear}
    alphaScaleBiasClamp {scale=0}
    boundBone {boneName=arm_tube_hand}
    physicsBodyDefinitions [{
      sphereCollisionRadius=0
      boundBone {boneName=arm_tube_hand}
      boxExtents=<10 20 20>
      constraintSetup {angularLimitsMax=<0 50 50>, angularLimitsMin=<0 -50 -50>, angularTarget=<0 0 0>}
      localJointOffset=<10 0 0>
     }]
   }
   n43 = Anim Dynamics - Bone: neck_tube_mid(ComponentPose=@n42, Alpha=1, ExternalForce="0,0,0") {
    angularDampingOverride=1
    angularSpringConstant=1500
    bAngularSpring=true
    bGravityOverrideInSimSpace=true
    bOverrideAngularDamping=true
    gravityScale=0.35
    simulationSpace=RootRelative
    alphaBoolBlend {blendOption=Linear}
    alphaScaleBiasClamp {scale=0.4}
    boundBone {boneName=neck_tube_mid}
    physicsBodyDefinitions [{
      sphereCollisionRadius=0
      boundBone {boneName=neck_tube_mid}
      boxExtents=<10 20 20>
      constraintSetup {angularLimitsMax=<0 30 0>, angularLimitsMin=<0 -30 -60>, angularTarget=<0 0 0>}
      localJointOffset=<10 0 0>
     }]
   }
   n44 = Anim Dynamics - Bone: arm_tube_mid(ComponentPose=@n43, Alpha=1, ExternalForce="0,0,0") {
    angularDampingOverride=1
    angularSpringConstant=1500
    bAngularSpring=true
    bGravityOverrideInSimSpace=true
    bOverrideAngularDamping=true
    gravityScale=0.35
    simulationSpace=RootRelative
    alphaBoolBlend {blendOption=Linear}
    alphaScaleBiasClamp {scale=0.4}
    boundBone {boneName=arm_tube_mid}
    physicsBodyDefinitions [{
      sphereCollisionRadius=0
      boundBone {boneName=arm_tube_mid}
      boxExtents=<10 20 20>
      constraintSetup {angularLimitsMax=<0 90 90>, angularLimitsMin=<0 -90 -90>, angularTarget=<0 0 0>}
      localJointOffset=<10 0 0>
     }]
   }
   n45 = Get target_alpha() -> (target_alpha:double)
   n46 = "Transform (Modify) Bone - Bone: arm_tube_hand"(ComponentPose=@n44, Alpha=@n45, Translation="0,0,0", Rotation="-70,0,0", Scale="1,1,1") {
    rotationMode=BMM_Additive
    rotationSpace=BCS_BoneSpace
    alphaBoolBlend {blendOption=Linear}
    alphaScaleBiasClamp {bInterpResult=true, interpSpeedDecreasing=5}
    boneToModify {boneName=arm_tube_hand}
    rotation {pitch=-70}
   }
   n47 = "Transform (Modify) Bone - Bone: hand_tube_mid"(ComponentPose=@n46, Alpha=@n45, Translation="0,0,0", Rotation="0,40,0", Scale="1,1,1") {
    rotationMode=BMM_Additive
    rotationSpace=BCS_BoneSpace
    alphaBoolBlend {blendOption=Linear}
    alphaScaleBiasClamp {bInterpResult=true, interpSpeedDecreasing=5}
    boneToModify {boneName=hand_tube_mid}
    rotation {yaw=40}
   }
   n48 = "Transform (Modify) Bone - Bone: arm_tube_mid"(ComponentPose=@n47, Alpha=@n45, Translation="0,0,0", Rotation="0,40,0", Scale="1,1,1") {
    rotationMode=BMM_Additive
    rotationSpace=BCS_BoneSpace
    alphaBoolBlend {blendOption=Linear}
    alphaScaleBiasClamp {bInterpResult=true, interpSpeedDecreasing=5}
    boneToModify {boneName=arm_tube_mid}
    rotation {yaw=40}
   }
   n49 = Anim Dynamics - Bone: belt_torso(ComponentPose=@n48, Alpha=1, ExternalForce="0,0,0") {
    angularDampingOverride=1
    angularSpringConstant=1500
    bAngularSpring=true
    bGravityOverrideInSimSpace=true
    bOverrideAngularDamping=true
    gravityScale=0.35
    simulationSpace=RootRelative
    alphaBoolBlend {blendOption=Linear}
    alphaScaleBiasClamp {scale=0.3}
    boundBone {boneName=belt_torso}
    physicsBodyDefinitions [{
      sphereCollisionRadius=0
      boundBone {boneName=belt_torso}
      boxExtents=<10 20 20>
      constraintSetup {angularLimitsMax=<0 50 50>, angularLimitsMin=<0 -45 -60>, angularTarget=<0 0 0>}
      localJointOffset=<10 0 0>
     }]
   }
   n50 = "Constraint - dyn_parent_belt_01 (PPPPPPPP)"!(
    ComponentPose=@n49
    Alpha=1
    ConstraintWeights_0=0.0042
    ConstraintWeights_1:float=0.1807
    ConstraintWeights_2:float=0.2424
    ConstraintWeights_3:float=0.3571
    ConstraintWeights_4:float=0.1541
    ConstraintWeights_5:float=0.0037
    ConstraintWeights_6:float=0.046
    ConstraintWeights_7:float=0.0118
   ) -> (Pose) {
    alphaBoolBlend {blendOption=Linear}
    boneToModify {boneName=dyn_parent_belt_01}
    constraintSetup [
     {targetBone {boneName=thigh_twist_01_r}}
     {targetBone {boneName=spine_01}}
     {targetBone {boneName=spine_02}}
     {targetBone {boneName=spine_03}}
     {targetBone {boneName=spine_04}}
     {targetBone {boneName=spine_05}}
     {targetBone {boneName=belt_torso}}
     {targetBone {boneName=pelvis}}
    ]
    constraintWeights [1,1,1,1,1,1,1,1]
   }
   Save cached pose 'RBAN_input_pose'(Pose=Component To Local(ComponentPose=@n50)) {cacheName=RBAN_input_pose}
   Save cached pose 'InputCache'(Pose=Copy Pose From Mesh() {bCopyCurves=true, bUseAttachedParent=true}) {cacheName=InputCache}
  ]}
 EventGraph {nodes [
   ResetDynamics(InTeleportType=TeleportPhysics)
   > ResetDynamics(InTeleportType=ResetPhysics)
   n1 = Event BlueprintUpdateAnimation()
   > n2 = Sequence!(execute) -> (then_0, then_1, then_2:Exec, then_3:Exec, then_4:Exec, then_5:Exec)
   n3 = Break Fort Anim Input Slope Info!(FortAnimInput_SlopeInfo:"Fort Anim Input Slope Info Structure (by ref)"=Get SlopeInfoInput() -> (SlopeInfoInput:Fort Anim Input Slope Info Structure)) -> (RootSlopeTranslation:Vector=~, LeftFootSlopeTranslation:Vector=~, RightFootSlopeTranslation:Vector=~, SlopeRotation:Rotator=~) {is=title K2Node_BreakStruct}
   n4 = BreakVector(InVec=@n3)
   n5 = BreakVector(InVec=VInterpTo!(Current:Vector=Get slope_foot_trans(), Target:Vector=@n3, DeltaTime:float=@n1.DeltaTimeX, InterpSpeed:float=4) -> (ReturnValue:Vector))
   > Set slope_foot_trans(slope_foot_trans:Vector=MakeVector(
     X=@n4
     Y=@n4.Y
     Z="Min (Float)"(A:double=@n5.Z, B:double=@n4.Z) -> (ReturnValue:double) {is=title K2Node_CommutativeAssociativeBinaryOperator}
    )) -> (Output_Get:Vector)
   > Set slope_foot_rot(slope_foot_rot:Rotator=RInterpTo!(Current:Rotator=Get slope_foot_rot(), Target:Rotator=@n3.SlopeRotation, DeltaTime:float=@n1.DeltaTimeX, InterpSpeed:float=2) -> (ReturnValue:Rotator)) -> (Output_Get:Rotator)
   Set target_alpha(execute=@n2.then_1, target_alpha:double="To Float (Boolean)"(InBool=Get bIsTargeting())) -> (Output_Get:double)
   n6 = Get skelmesh() -> (skelmesh:Skeletal Mesh Component Object Reference)
   Set head_back_alpha(
    execute=@n2.then_2
    head_back_alpha:double=GetAlphaFromReferenceTransformDelta(SkeletalMeshComponent=@n6, BoneName=head, BaseName=spine_05, DeltaComponentType=QuaternionTwist, InputRange=(X=10,Y=40), TwistAxis="0,0,1")
   ) -> (Output_Get:double)
   n7 = GetAlphaFromReferenceTransformDelta(SkeletalMeshComponent=@n6, BoneName=head, BaseName=spine_05, DeltaComponentType=QuaternionTwist, InputRange=(X=-10,Y=-30), TwistAxis="0,0,1")
   > Set head_right_alpha(head_right_alpha:double="Max (Float)"(
     A=@n7
     B=GetAlphaFromReferenceTransformDelta(SkeletalMeshComponent=@n6, BoneName=head, BaseName=spine_05, DeltaComponentType=QuaternionTwist, InputRange=(X=0,Y=-40), TwistAxis="1,0,0")
    )) -> (Output_Get:double)
   > Set head_left_alpha(head_left_alpha:double="Max (Float)"(
     A=@n7
     B=GetAlphaFromReferenceTransformDelta(SkeletalMeshComponent=@n6, BoneName=head, BaseName=spine_05, DeltaComponentType=QuaternionTwist, InputRange=(X=0,Y=40), TwistAxis="1,0,0")
    )) -> (Output_Get:double)
   > Set thigh_fwd_l_alpha(thigh_fwd_l_alpha:double=GetAlphaFromReferenceTransformDelta(SkeletalMeshComponent=@n6, BoneName=thigh_l, BaseName=pelvis, DeltaComponentType=QuaternionTwist, InputRange=(X=20,Y=100), TwistAxis="0,0,1")) -> (Output_Get:double)
   n8 = Branch(execute=@n2.then_3, Condition=Get bIsSwimSprinting() -> (bIsSwimSprinting:bool))
   > Set GravityOverride_b(GravityOverride_b:Vector="0, 0, 0") -> (Output_Get:Vector)
   n9 = Branch(execute=@n8.else, Condition=Get bIsSkyDiving())
   > n10 = Branch(Condition=Get bIsParachuteOpened())
   Set GravityOverride_b(execute=[@n9.else, @n10.then], GravityOverride_b="0, 0,-920")
   n11 = Branch(execute=@n10.else, Condition="float > float"(A=Get PawnVelocityZ(), B=50) {is=title K2Node_CallFunction})
   > Set GravityOverride_b(GravityOverride_b="0, 0,200")
   n12 = PerlinNoise1D!(Value:float=GetGameTimeInSeconds()) -> (ReturnValue:float)
   n13 = float * float(A="Sin (Degrees)"!(A:double=float * float(A=@n12, B=250)) -> (ReturnValue:double), B=-100)
   n14 = Get Skel_Mesh()
   n15 = GetSocketRotation!(self:Scene Component Object Reference=@n14, InSocketName:Name=pelvis) -> (ReturnValue:Rotator)
   n16 = vector + vector(
    A:Vector=vector * float(A=GetForwardVector!(InRot:Rotator=@n15) -> (ReturnValue:Vector), B=-800)
    B:Vector=MakeVector(Y=0, Z=float + float(A=@n13, B=200) {is=title K2Node_CommutativeAssociativeBinaryOperator})
   ) -> (ReturnValue:Vector) {is=title K2Node_CommutativeAssociativeBinaryOperator}
   Set GravityOverride_b(execute=@n11.else, GravityOverride_b=@n16)
   n17 = Branch(execute=@n2.then_4, Condition=Get bIsOstrichVehicleDriver() -> (bIsOstrichVehicleDriver:bool))
   Set in vehicle rigid alpha(execute=[@n17.then, @n18.then, @n19.then], in vehicle rigid alpha:double=0) -> (Output_Get:double)
   > Set in vehicle scale alpha(in vehicle scale alpha:double=1) -> (Output_Get:double)
   n18 = Branch(execute=@n17.else, Condition=Get bIsFerretVehicleDriver() -> (bIsFerretVehicleDriver:bool))
   n19 = Branch(execute=@n18.else, Condition=Get bIsOctopusVehicleDriver() -> (bIsOctopusVehicleDriver:bool))
   Set in vehicle rigid alpha(execute=@n19.else, in vehicle rigid alpha=1)
   > Set in vehicle scale alpha(in vehicle scale alpha=0)
   Event OnCharacterCustomizationLoaded()
   > Set skelmesh(skelmesh:Skeletal Mesh Component Object Reference=GetOwnerSkeletalMeshComponent()) -> (Output_Get:Skeletal Mesh Component Object Reference)
   Event BlueprintInitializeAnimation()
   n20 = GetOwnerSkeletalMeshComponent()
   > Is Valid(InputObject=@n20)
   > Set Skel_Mesh(Skel_Mesh=@n20)
   Set head_right_alpha(head_right_alpha=GetAlphaFromReferenceTransformDelta(SkeletalMeshComponent=@n6, BoneName=head, BaseName=spine_05, DeltaComponentType=QuaternionTwist, InputRange=(X=0,Y=-50), TwistAxis="1,0,0"))
   > Set head_left_alpha(head_left_alpha=GetAlphaFromReferenceTransformDelta(SkeletalMeshComponent=@n6, BoneName=head, BaseName=spine_05, DeltaComponentType=QuaternionTwist, InputRange=(X=0,Y=50), TwistAxis="1,0,0"))
   GetVectorArrayAverage!(Vectors:Array of Vectors=Make Array!("[0]":Vector=@n3.LeftFootSlopeTranslation, "[1]":Vector=@n3.RightFootSlopeTranslation) -> (Array:Array of Vectors)) -> (ReturnValue:Vector)
   n21 = GetSocketLocation!(self:Scene Component Object Reference=@n14, InSocketName:Name=pelvis) -> (ReturnValue:Vector)
   DrawDebugLine!(
    execute
    LineStart:Vector=@n21
    LineEnd:Vector=vector + vector(A=@n21, B=@n16)
    LineColor:Linear Color Structure=(R=1,G=0,B=0,A=1)
    Duration:float=0
    Thickness:float=3
    DepthPriority:EDrawDebugSceneDepthPriorityGroup Enum=World
   ) -> (then)
   Event OnEndCrouching()
  ]}
]
variables [
 add_debug {type=number, default=1}
 GravityOverride_b {type=Vector}
 head_back_alpha {type=number}
 head_left_alpha {type=number}
 head_right_alpha {type=number}
 in vehicle rigid alpha {type=number}
 in vehicle scale alpha {type=number}
 jacket_material {type=MaterialInstanceDynamic}
 Skel_Mesh {type=SkeletalMeshComponent}
 skelmesh {type=SkeletalMeshComponent}
 slope_foot_r_trans {type=Vector}
 slope_foot_rot {type=Rotator}
 slope_foot_trans {type=Vector}
 target_alpha {type=number}
 thigh_fwd_l_alpha {type=number}
]
