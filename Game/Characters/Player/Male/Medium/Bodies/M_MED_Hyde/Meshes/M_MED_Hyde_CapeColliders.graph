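# Collision-only physics asset for the M_MED_Hyde medium male player body, named as the colliders for its cape.
# Overlapping capsules cover pelvis, spine, clavicles and neck, with a sphere on the head.
# Extra capsules on the back arm tube and two long ones on the rear dynamic coat bones add volume beyond the core body.
# Arm, leg and foot bodies have no shapes, and no constraints link any bodies.
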PhysicsAsset /Game/Characters/Player/Male/Medium/Bodies/M_MED_Hyde/Meshes/M_MED_Hyde_CapeColliders
bodies {
 arm_tube_back [{
   shapeType=Capsule
   length=15
   radius=15
   shapeName=arm_tube_back_capsule
   center=<-10.5476 16.8518 -6.0896>
   rotation=<rot -59.9999 0.0006 -30.0005>
  }]
 calf_l []
 calf_r []
 clavicle_l [{shapeType=Capsule, length=7.7636, radius=11.1427, center=<9.9642 -0.7741 -1.5923>, rotation=<rot 75.9352 -0.0001 175.5622>}]
 clavicle_r [{shapeType=Capsule, length=7.7636, radius=11.1427, center=<-9.9642 0.7741 1.5923>, rotation=<rot 75.9351 -0.0001 -4.4378>}]
 dyn_coat_bkl_05 [{
   shapeType=Capsule
   length=60.75
   radius=15
   shapeName=dyn_coat_bkl_05_capsule
   center=<16.3511 10.4424 -9.1905>
   rotation=<rot -79.9999 180 -180>
  }]
 dyn_coat_bkr_05 [{shapeType=Capsule, length=60.75, radius=15, shapeName=dyn_coat_bkr_05_capsule, center=<16.3511 -9.3311 -9.1905>, rotation=<rot -80 -180 180>}]
 foot_l []
 foot_r []
 head [{shapeType=Sphere, radius=17.3125, center=<11.028 -0.174 0>}]
 lowerarm_l []
 lowerarm_r []
 neck_02 [{shapeType=Capsule, length=13.4375, radius=7.3175, center=<-0.2642 -0.3408 0>, rotation=<rot 90 -171.1193 -175.3978>}]
 pelvis [{shapeType=Capsule, length=6.6562, radius=15.5388, center=<-2.6215 -1.3695 0>}]
 spine_01 [{shapeType=Capsule, length=9.7187, radius=15.4138, center=<-2.6215 -0.7156 0>}]
 spine_02 [
  {shapeType=Capsule, length=5.3437, radius=16.945, center=<-2.6215 0.4993 0.0001>}
  {shapeType=Capsule, length=29.875, radius=11.1013, center=<2.5024 0.4992 3.9825>, rotation=<rot 90 0.0049 0.0049>}
  {shapeType=Capsule, length=29.875, radius=11.1013, center=<2.4704 0.4991 -3.9825>, rotation=<rot 90 -0.0396 -0.0396>}
 ]
 spine_03 [{shapeType=Capsule, length=10.6562, radius=16.7575, center=<1.2448 0.1364 0.0001>}]
 spine_04 [{shapeType=Capsule, length=7.9062, radius=20.3825, center=<3.5267 -1.2748 0.0002>}]
 spine_05 [
  {shapeType=Capsule, length=13.8437, radius=21.195, center=<-1.7514 -1.0463 0>}
  {shapeType=Capsule, length=8.5312, radius=21.3125, center=<2.0987 -3.35 0>}
 ]
 thigh_l []
 thigh_r []
 upperarm_l []
 upperarm_r []
}
constraints []
