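# Animation Blueprint for Hyde character mesh parts, adding secondary motion to hat and mutton bones over the pose copied from the attached parent.
# Chained AnimDynamics nodes spring those bones, then additive bone transforms rotate each mutton bone.
# Head twist relative to spine_05 drives those corrective alphas, taking the larger of downward and sideways values.
# After evaluation, a zero-delay step sends the hat_1 location to the part's materials as HatRootLocation.
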AnimBlueprint /Game/Characters/Player/Male/Medium/Bodies/M_MED_Hyde/Meshes/Parts/M_MED_Hyde_AnimBP
target_skeleton=&/M_MED_Hyde_Skeleton
graphs [
 AnimGraph {nodes [
   n1 = Anim Dynamics - Bone: mutton_r(
    ComponentPose=Local To Component(LocalPose=Copy Pose From Mesh() {bCopyCurves=true, bUseAttachedParent=true})
    Alpha=0.3
    ExternalForce="0,0,0"
   ) {
    angularDampingOverride=1
    angularSpringConstant=1500
    bAngularSpring=true
    bGravityOverrideInSimSpace=true
    bOverrideAngularDamping=true
    gravityScale=0.35
    simulationSpace=RootRelative
    alphaBoolBlend {blendOption=Linear}
    boundBone {boneName=mutton_r}
    physicsBodyDefinitions [{
      sphereCollisionRadius=0
      boundBone {boneName=mutton_r}
      boxExtents=<10 20 20>
      constraintSetup {angularLimitsMax=<0 40 40>, angularLimitsMin=<0 -40 -40>, angularTarget=<0 0 0>}
      localJointOffset=<10 0 0>
     }]
   }
   n2 = Anim Dynamics - Bone: mutton_l(ComponentPose=@n1, Alpha=0.3, ExternalForce="0,0,0") {
    angularDampingOverride=1
    angularSpringConstant=1500
    bAngularSpring=true
    bGravityOverrideInSimSpace=true
    bOverrideAngularDamping=true
    gravityScale=0.35
    simulationSpace=RootRelative
    alphaBoolBlend {blendOption=Linear}
    boundBone {boneName=mutton_l}
    physicsBodyDefinitions [{
      sphereCollisionRadius=0
      boundBone {boneName=mutton_l}
      boxExtents=<10 20 20>
      constraintSetup {angularLimitsMax=<0 40 40>, angularLimitsMin=<0 -40 -40>, angularTarget=<0 0 0>}
      localJointOffset=<10 0 0>
     }]
   }
   n3 = Anim Dynamics - Bone: hat_tatter(ComponentPose=@n2, Alpha=1, ExternalForce="0,0,0") {
    angularDampingOverride=1
    angularSpringConstant=1500
    bAngularSpring=true
    bGravityOverrideInSimSpace=true
    bOverrideAngularDamping=true
    gravityScale=0.35
    simulationSpace=RootRelative
    alphaBoolBlend {blendOption=Linear}
    boundBone {boneName=hat_tatter}
    physicsBodyDefinitions [{
      sphereCollisionRadius=0
      boundBone {boneName=hat_tatter}
      boxExtents=<10 20 20>
      constraintSetup {angularLimitsMax=<0 0 0>, angularLimitsMin=<0 0 -40>, angularTarget=<0 0 0>}
      localJointOffset=<10 0 0>
     }]
   }
   n4 = Anim Dynamics - Bone: hat_1(ComponentPose=@n3, Alpha=0.1, ExternalForce="0,0,0") {
    angularDampingOverride=1
    angularSpringConstant=1500
    bAngularSpring=true
    bGravityOverrideInSimSpace=true
    bOverrideAngularDamping=true
    gravityScale=0.35
    simulationSpace=RootRelative
    alphaBoolBlend {blendOption=Linear}
    boundBone {boneName=hat_1}
    physicsBodyDefinitions [{
      sphereCollisionRadius=0
      boundBone {boneName=hat_1}
      boxExtents=<10 20 20>
      constraintSetup {angularLimitsMax=<0 40 40>, angularLimitsMin=<0 -40 -40>, angularTarget=<0 0 0>}
      localJointOffset=<10 0 0>
     }]
   }
   n5 = Anim Dynamics - Bone: hat_2(ComponentPose=@n4, Alpha=0.5, ExternalForce="0,0,0") {
    angularDampingOverride=1
    angularSpringConstant=1500
    bAngularSpring=true
    bGravityOverrideInSimSpace=true
    bOverrideAngularDamping=true
    gravityScale=0.35
    simulationSpace=RootRelative
    alphaBoolBlend {blendOption=Linear}
    boundBone {boneName=hat_2}
    physicsBodyDefinitions [{
      sphereCollisionRadius=0
      boundBone {boneName=hat_2}
      boxExtents=<10 20 20>
      constraintSetup {angularLimitsMax=<0 0 0>, angularLimitsMin=<0 0 -40>, angularTarget=<0 0 0>}
      localJointOffset=<10 0 0>
     }]
   }
   n6 = Anim Dynamics - Bone: hat_3(ComponentPose=@n5, Alpha=0.5, ExternalForce="0,0,0") {
    angularDampingOverride=1
    angularSpringConstant=1500
    bAngularSpring=true
    bGravityOverrideInSimSpace=true
    bOverrideAngularDamping=true
    gravityScale=0.35
    simulationSpace=RootRelative
    alphaBoolBlend {blendOption=Linear}
    boundBone {boneName=hat_3}
    physicsBodyDefinitions [{
      sphereCollisionRadius=0
      boundBone {boneName=hat_3}
      boxExtents=<10 20 20>
      constraintSetup {angularLimitsMax=<0 0 0>, angularLimitsMin=<0 0 -40>, angularTarget=<0 0 0>}
      localJointOffset=<10 0 0>
     }]
   }
   n7 = "Transform (Modify) Bone - Bone: mutton_r"(ComponentPose=@n6, Alpha=Get head_right_alpha() -> (head_right_alpha:double), Translation="0,0,0", Rotation="-20,-50,0", Scale="1,1,1") {
    rotationMode=BMM_Additive
    rotationSpace=BCS_BoneSpace
    alphaBoolBlend {blendOption=Linear}
    alphaScaleBiasClamp {bInterpResult=true}
    boneToModify {boneName=mutton_r}
    rotation {pitch=-20, yaw=-50}
   }
   n8 = "Transform (Modify) Bone - Bone: mutton_l"(ComponentPose=@n7, Alpha=Get head_left_alpha() -> (head_left_alpha:double), Translation="0,0,0", Rotation="20,-50,0.000", Scale="1,1,1") {
    rotationMode=BMM_Additive
    rotationSpace=BCS_BoneSpace
    alphaBoolBlend {blendOption=Linear}
    alphaScaleBiasClamp {bInterpResult=true}
    boneToModify {boneName=mutton_l}
    rotation {pitch=20, yaw=-50}
   }
   Output Pose(Result=Component To Local(ComponentPose=@n8)) {layerGroup=None}
   Get head_down_alpha() -> (head_down_alpha:double)
   Get head_down_alpha()
  ]}
 EventGraph {nodes [
   Event BlueprintUpdateAnimation()
   n1 = Get skelmesh() -> (skelmesh:Skeletal Mesh Component Object Reference)
   n2 = GetAlphaFromReferenceTransformDelta(SkeletalMeshComponent=@n1, BoneName=head, BaseName=spine_05, DeltaComponentType=QuaternionTwist, InputRange=(X=7,Y=-20), TwistAxis="0,0,1")
   > Set head_down_alpha(head_down_alpha:double=@n2) -> (Output_Get:double)
   > Set head_right_alpha(head_right_alpha:double="Max (Float)"(
     A=@n2
     B=GetAlphaFromReferenceTransformDelta(SkeletalMeshComponent=@n1, BoneName=head, BaseName=spine_05, DeltaComponentType=QuaternionTwist, InputRange=(X=0,Y=-50), TwistAxis="1,0,0")
    )) -> (Output_Get:double)
   > Set head_left_alpha(head_left_alpha:double="Max (Float)"(
     A=@n2
     B=GetAlphaFromReferenceTransformDelta(SkeletalMeshComponent=@n1, BoneName=head, BaseName=spine_05, DeltaComponentType=QuaternionTwist, InputRange=(X=0,Y=50), TwistAxis="1,0,0")
    )) -> (Output_Get:double)
   Event OnCharacterCustomizationLoaded()
   > Set skelmesh(skelmesh:Skeletal Mesh Component Object Reference=GetOwnerSkeletalMeshComponent()) -> (Output_Get:Skeletal Mesh Component Object Reference)
   Event BlueprintPostEvaluateAnimation() {is=title K2Node_Event}
   n3 = GetPartSkeletalMeshComponent()
   > Is Valid(InputObject=@n3)
   > Delay(Duration=0)
   > SetVectorParameterValueOnMaterials!(
    execute
    self:Mesh Component Object Reference=@n3
    ParameterName:Name=HatRootLocation
    ParameterValue:Vector=GetSocketLocation!(self:Scene Component Object Reference=@n3, InSocketName:Name=hat_1) -> (ReturnValue:Vector)
   ) -> (then)
  ]}
]
variables [
 head_down_alpha {type=number}
 head_left_alpha {type=number}
 head_right_alpha {type=number}
 skelmesh {type=SkeletalMeshComponent}
]
